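# Physics asset for the Pickaxe_Racer skeletal mesh, a Fortnite Battle Royale cosmetic two-handed melee pickaxe.
# It holds a single body on the Melee_TwoHanded_Handle bone, shaped as one large capsule offset along and rotated to fit the tool.
# There are no constraints, so the whole pickaxe is represented as one rigid collision volume.
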PhysicsAsset /BRCosmetics/Weapons/FORT_Melee/Pickaxe_Racer/Meshes/SK_Pickaxe_Racer_PhysicsAsset
bodies {Melee_TwoHanded_Handle [{shapeType=Capsule, length=73.1578, radius=38.5908, center=<0.7928 -0.9834 27.6057>, rotation=<rot 0.0161 -90 -0.0087>}]}
constraints []
